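# Niagara skydiving contrail cosmetic (HollowStun) that trails an icy blue effect from skeletal mesh bones or sockets.
# Mesh emitters draw fire-trail, long-trail and spiral layers, while ice-fragment meshes and curl-noise snowflake sprites spawn continuously around them.
# ParaGlideLeanAlpha drives trail scale, spawn rates and speeds; front-end modifiers and a timeline adjust orientation for preview, and time-of-day colors tint several layers.
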NiagaraSystem /BRCosmetics/Athena/Effects/Particles/Skydiving/NS_Contrail_HollowStun/NS_Contrail_HollowStun
dependencies {
 usedDataInterfaces [
  &NiagaraDataInterfaceCurve
  &NiagaraDataInterfaceArrayInt32
  &NiagaraDataInterfaceColorCurve
  &NiagaraDataInterfaceArrayFloat3
  &NiagaraDataInterfaceVectorCurve
  &NiagaraDataInterfacePlatformSet
  &NiagaraDataInterfaceMeshRendererInfo
  &NiagaraDataInterfaceArrayFloat
  &NiagaraDataInterfaceSpriteRendererInfo
  &NiagaraDataInterfaceSkeletalMesh
 ]
 usedDynamicInputs [
  &ClampFloat
  &RemapRange
  &MakeFloatFromVector
  &SimulationPosition
  &UniformRangedVector2D
  &FN_FrontEndVector2Modifier
 ]
 usedModules [
  &self:SystemSpawnScript.SystemScriptSource.SystemScriptGraph.NiagaraNodeAssignment_0.SetVariables_#1
  &SystemState
  &self:UpdatePreviousRotation
  &self:SystemSpawnScript.SystemScriptSource.SystemScriptGraph.NiagaraNodeAssignment_5.SetVariables_#2
  &self:SystemSpawnScript.SystemScriptSource.SystemScriptGraph.NiagaraNodeAssignment_6.SetVariables_#3
  /Game/Effects/Niagara/Modules/Contrails/FN_Contrail_Setup
  &Timeline
 ]
}
emitters {
 FireTrail_And_Dots {
  emitterUpdateScript [
   &EmitterState(
    Inactive Response=::NewEnumerator1
    Life Cycle Mode=::NewEnumerator1
    Loop Behavior=::NewEnumerator1
    Loop Duration Mode=::NewEnumerator1
    UseLoopDelay=false
   )
   &SpawnBurst_Instantaneous(Spawn Count=2, Use Spawn Probability=false, UseLoopCountLimit=false)
  ]
  particleSpawnScript [
   &PartitionParticles()
   &InitializeParticle(
    Color=<rgba 0.0442 0.1096 0.2063 2>
    Lifetime=3
    Mesh Scale=&FN_FrontEndVector3Modifier(Gameplay=<0.5 0.5 1.8>, Front End=<0.5 0.5 1.8>)
    Mesh Scale Mode=::NewEnumerator4
    UsePositionOffset=false
    Write Material Random=true
   )
   &InitialMeshOrientation(
    Mesh Axis To Orient=<0 0 1>
    Mesh Orientation Mode=::NewEnumerator2
    Orientation Coordinate Space=::World
    Orientation Vector=&FN_FrontEndVector3Modifier(Gameplay=<0 0 1>, Front End=<0 0 1>)
    Rotation=&FN_FrontEndVector3Modifier(Gameplay=<0 0 0>, Front End=<-0.04 0 0>)
    Rotation Coordinate Space=::NewEnumerator0
    Use Rotation Vector=true
   )
   &CameraOffset(Camera Offset Amount=-20)
   /SetVariables() {isSetParameters=true}
  ]
  particleUpdateScript [
   &ParticleState(Kill Particles When Lifetime Has Elapsed=false, Let Infinitely Lived Particles Die When Emitter Deactivates=true)
   &SkeletalMeshLocation(
    Bone / Socket Index=$Particles.PartitionParticles.Partition
    Bone Sampling Mode=::NewEnumerator4
    MaskSpawnGroup=false
    Orientation Sampling=::NewEnumerator1
    Sampled Position Offset=&Lerp_Vector(A=<0 0 -10>, B=<0 0 -10>, Alpha=$User.ParaGlideLeanAlpha)
    Sampled Position Offset Coordinate Space:ENiagaraSkelMeshTransforms=::NewEnumerator2
    Skeletal Mesh=$System.Skeletal Mesh
   )
   &ScaleMeshSize(Scale Factor=&MakeVector(
     X=1
     Y=1
     Z=&FloatFromCurve(
      FloatCurve=DataInterface() {propertyValues {Curve {keys [<at 0.5 0 RCIM_Cubic>, {arriveTangent=0.2452, interpMode=RCIM_Cubic, leaveTangent=0.2452, tangentMode=RCTM_User, time=1, value=1}]}}}
      CurveIndex=$Engine.System.Age
      Scale Curve=&Lerp_Float(A=&FN_FrontEndFloatModifier(Gameplay=1, Front End=1.1), B=1.2, Alpha=$User.ParaGlideLeanAlpha)
     )
    )) {module=ScaleMeshSize001}
   /SetVariables(Particles.Position=&AddVectorToPosition(Position=$Particles.Position, Vector=&TransformVector(Vector=<0 0 0>, Destination Space=::Simulation))) {enabled=false, isSetParameters=true}
   &InterpolateOverTime(Rate Of Change=5, Target Vector Value:NiagaraPosition=$Particles.Position, Variable Type=::NewEnumerator2) {enabled=false}
   /SetVariables(Particles.TrailDynOffset:NiagaraFloat=&Add_Float(
     A=$Particles.TrailDynOffset
     B=&Multiply_Float(A=$Engine.DeltaTime, B=&Lerp_Float(A=0, B=-0.5, Alpha=$User.ParaGlideLeanAlpha))
    )) {enabled=false, isSetParameters=true}
   &DynamicMaterialParameters(
    Index 0 Param 1=&Multiply_Float(
     A=&FN_FrontEndFloatModifier(Gameplay=50, Front End=45)
     B=&FloatFromCurve(
      FloatCurve=DataInterface() {propertyValues {Curve {keys [<at 0.5 0 RCIM_Cubic>, {interpMode=RCIM_Cubic, leaveTangent=3.362, time=1.5, value=1}]}}}
      CurveIndex=$System.Age
     )
    )
    Index 0 Param 2=&Multiply_Float(
     A=&FN_FrontEndFloatModifier(Gameplay=80, Front End=80)
     B=&FloatFromCurve(
      FloatCurve=DataInterface() {propertyValues {Curve {keys [<at 0.5 0 RCIM_Cubic>, {arriveTangent=1.9352, interpMode=RCIM_Cubic, leaveTangent=1.9352, tangentMode=RCTM_User, time=1.5, value=1}]}}}
      CurveIndex=$System.Age
     )
    )
    Index 0 Param 3=1
    Index 0 Param 4=&Add_Float(A=&RandomRangeFloat(Minimum=0, Maximum=100), B=$Particles.TrailDynOffset)
    Module.Index0Param0Write=true
    Module.Index0Param1Write=true
    Module.Index0Param2Write=true
    Module.Index0Param3Write=true
   )
   &ScaleColor(
    Scale RGB=&MakeVectorFromLinearColorRGB(Color=&LerpLinearColors(StartColor=<rgba 1 1 1 1>, EndColor=<rgba 0.75 0.525 0.525 1>, LerpFactor=$User.ParaGlideLeanAlpha))
    ScaleA=true
    ScaleRGB=true
   )
   /SetVariables(Particles.MeshOrientation=&OrientMeshToVector(
     Look At Direction=&TransformVector(Vector=&Lerp_Vector(A=<0.1 0 0>, B=<1 0 0>, Alpha=$User.ParaGlideLeanAlpha), Destination Space=::World)
     Reference Axis Vector=<1 0 0>
    )) {enabled=false, isSetParameters=true}
   &UpdateMeshOrientation(
    Facing Interpolation=false
    Orientation Method=::NewEnumerator1
    Override Quat=false
    TransformFacing=false
    TransformUp=false
    Up Direction:Vector3f=&FN_FrontEndVector3Modifier(
     Gameplay=<1 0 0>
     Front End=&Lerp_Vector(
      A=<1 -0.2 1>
      B=<1 -0.9 1>
      Alpha=&FN_FrontEndFloatModifier(Gameplay=0, Front End=$System.Timeline.NormalizedTime)
     )
    )
   )
  ]
  rendererClasses [&NiagaraMeshRendererProperties]
  renderers [{rendererClass=&NiagaraMeshRendererProperties, rendererIndex=0}]
 }
 Ice_Frags {
  emitterUpdateScript [
   &EmitterState()
   &SpawnRate(
    SpawnRate=&FloatFromCurve(
     FloatCurve=DataInterface() {propertyValues {Curve {
        defaultValue=1
        keys [
         {arriveTangent=4.1785, interpMode=RCIM_Cubic, leaveTangent=4.1785, tangentMode=RCTM_User, time=0, value=0}
         {arriveTangent=1.1179, interpMode=RCIM_Cubic, leaveTangent=1.1179, tangentMode=RCTM_User, time=0.2042, value=0.6228}
         {arriveTangent=0.0224, interpMode=RCIM_Cubic, leaveTangent=0.0224, tangentMode=RCTM_User, time=1, value=1}
        ]
       }}}
     CurveIndex=$Emitter.NormalizedLoopAge
     Scale Curve=&Lerp_Float(A=40, B=80, Alpha=$User.ParaGlideLeanAlpha)
    )
    Use Spawn Probability=false
   )
  ]
  particleSpawnScript [
   &PartitionParticles()
   &InitializeParticle(
    Color=&MakeCustomLinearColorFromBool(Bool=&RandomBool(), True Color=<rgba 0.4995 1.0577 1.5 1>, False Color=<rgba 0.1213 0.3179 1.5 1>)
    Lifetime Max=0.75
    Lifetime Min=0.3
    Lifetime Mode=::NewEnumerator1
    Mesh Renderer Array Visibility Mode=::NewEnumerator1
    Mesh Renderer Info=DataInterface() {propertyValues {MeshRenderer=&self:Ice_Frags_0.NiagaraMeshRendererProperties_1}}
    Mesh Renderer Mesh Index:NiagaraInt32=&RandomRangeInt(Minimum=0, Maximum=2)
    Mesh Scale Max=<0.2 0.2 0.3>
    Mesh Scale Min=<0.1 0.1 0.15>
    Mesh Scale Mode=::NewEnumerator2
    Sprite Size Mode=::NewEnumerator1
    Uniform Sprite Size Max=20
    UsePositionOffset=false
    Write Material Random=true
   )
   &SkeletalMeshLocation(
    Bone / Socket Index=$Particles.PartitionParticles.Partition
    Bone Sampling Mode=::NewEnumerator4
    MaskSpawnGroup=false
    Skeletal Mesh=$System.Skeletal Mesh
   )
   &CylinderLocation(Cylinder Height=5, Cylinder Radius=5, MaskSpawnGroup=false, Override Seed=false, Surface Only=true, UseOwnerScale=true)
   &CameraOffset(Camera Offset Amount=&RandomRangeFloat(Minimum=0, Maximum=20, Evaluation Type=::NewEnumerator1))
   &AddVelocity(
    Rotation Mode=::NewEnumerator2
    Velocity=<0 0 150>
    Yaw / Pitch / Roll:Vector3f=&Lerp_Vector(
     A=<0 0 0>
     B=<0 0 -45>
     Alpha=&FN_FrontEndFloatModifier(Gameplay=0, Front End=$System.Timeline.NormalizedTime)
    )
   )
   &InitialMeshOrientation(
    Mesh Axis To Orient=&SelectVectorFromArray(
     Direct Array Index=$Particles.PartitionParticles.Partition
     Vector Selection Array=DataInterface() {propertyValues {FloatData [<1 0 0>, <1 0 0>]}}
    )
    Mesh Orientation Mode=::NewEnumerator2
    Orientation Coordinate Space=::Local
    Orientation Vector=<0 0 1>
    Rotation=&RandomRangeVector(Minimum=<0 0 -0.1>, Maximum=<0 0 0.1>)
    Use Rotation Vector=true
   )
  ]
  particleUpdateScript [
   &ParticleState()
   /Niagara/Modules/Update/Forces/CurlNoiseForce(
    Noise Frequency=0.1
    Noise Strength=&SelectFloatFromArray(Direct Array Index=$Particles.PartitionParticles.Partition, Float Selection Array=DataInterface() {propertyValues {FloatData [200,200,0,0]}})
    Pan Noise=false
   )
   &UpdateMeshOrientation(
    Rotation Rate=&FloatFromCurve(
     FloatCurve=DataInterface() {propertyValues {Curve {defaultValue=0, keys [{tangentMode=RCTM_User, time=0, value=0.5}, {tangentMode=RCTM_User, time=1, value=1}]}}}
     Scale Curve=&Multiply_Float(
      A=&RandomRangeFloat(Minimum=1, Maximum=2)
      B=&Lerp_Float(A=1.5, B=2, Alpha=$User.ParaGlideLeanAlpha)
     )
    )
    Rotation Vector=&RandomRangeVector(Minimum=<-1 -0.5 -1>, Maximum=<1 0.5 1>)
   )
   &ScaleVelocity(Velocity Scale=&Lerp_Vector(A=<1 1 1.1>, B=<1 1 2>, Alpha=$System.Timeline.NormalizedTime))
   &SolveForcesAndVelocity(Clamp Velocity=false, Limit Acceleration=false)
   &ScaleMeshSize(Scale Factor=&VectorFromFloat(Value=&FloatFromCurve(FloatCurve=DataInterface() {propertyValues {Curve {keys [
          {arriveTangent=1.7195, interpMode=RCIM_Cubic, leaveTangent=1.7195, tangentMode=RCTM_User, time=0.2038, value=0.6697}
          {arriveTangent=-3.8833, interpMode=RCIM_Cubic, leaveTangent=1.7195, tangentMode=RCTM_Break, time=1.0051, value=0.2344}
         ]}}})))
   &ScaleColor(
    Scale Alpha=&FloatFromCurve(FloatCurve=DataInterface() {propertyValues {Curve {keys [<at 0.5 1 RCIM_Cubic>,<at 1.0011 0.002 RCIM_Cubic>]}}})
    Scale RGB=&MakeVectorFromLinearColorRGB(Color=&FN_Select_Color_from_Time_Of_Day(Morning=<rgba 0.3208 0.5045 0.8 1>, Day=<rgba 0.3125 0.5928 1 1>, Evening=<rgba 0.3542 0.6556 1 1>, Night=<rgba 0.2161 0.2929 0.5 1>))
    ScaleA=true
    ScaleRGB=true
   )
  ]
  rendererClasses [&NiagaraMeshRendererProperties]
  renderers [
   {rendererClass=&NiagaraMeshRendererProperties, rendererIndex=0}
   {rendererClass=&NiagaraMeshRendererProperties, rendererIndex=1}
   {rendererClass=&NiagaraMeshRendererProperties, rendererIndex=2}
  ]
 }
 Snowflakes {
  emitterUpdateScript [
   &EmitterState()
   &SpawnRate(
    SpawnRate=&FloatFromCurve(
     FloatCurve=DataInterface() {propertyValues {Curve {
        defaultValue=1
        keys [
         {arriveTangent=4.1785, interpMode=RCIM_Cubic, leaveTangent=4.1785, tangentMode=RCTM_User, time=0, value=0}
         {arriveTangent=1.1179, interpMode=RCIM_Cubic, leaveTangent=1.1179, tangentMode=RCTM_User, time=0.2042, value=0.6228}
         {arriveTangent=0.0224, interpMode=RCIM_Cubic, leaveTangent=0.0224, tangentMode=RCTM_User, time=1, value=1}
        ]
       }}}
     CurveIndex=$Emitter.NormalizedLoopAge
     Scale Curve=40
    )
    Use Spawn Probability=false
   )
  ]
  particleSpawnScript [
   &PartitionParticles()
   &InitializeParticle(
    AdjustAlpha=false
    AdjustHue=true
    AdjustSaturation=false
    AdjustValue=false
    Color=<rgba 0.4435 10.1588 20 1>
    Color Mode=::NewEnumerator3
    Lifetime=&Multiply_Float(
     A=&UniformRangedFloat(Minimum=0.3, Maximum=1.5)
     B=&SelectFloatFromArray(Direct Array Index=$Particles.SpawnGroup, Float Selection Array=DataInterface() {propertyValues {FloatData [1,1,0.5,0.5]}})
    )
    Mass Mode=::NewEnumerator1
    Sprite Size Mode=::NewEnumerator1
    Uniform Sprite Size Max=5
    Uniform Sprite Size Min=2
    UsePositionOffset=false
    Write Material Random=true
   ) {module=InitializeParticle001}
   &SkeletalMeshLocation(
    Bone / Socket Index=$Particles.PartitionParticles.Partition
    Bone Sampling Mode=::NewEnumerator4
    MaskSpawnGroup=false
    Skeletal Mesh=$System.Skeletal Mesh
   )
   &AddVelocity(
    Cone Angle=5
    Cone Axis=<0 0 1>
    Module.UseConeFalloff=false
    Module.UseVelDistribution=false
    Override Seed=false
    Rotation Mode=::NewEnumerator2
    Velocity Mode=::NewEnumerator2
    Velocity Speed=&RandomRangeFloat(Minimum=100, Maximum=200)
    Yaw / Pitch / Roll:Vector3f=&Lerp_Vector(
     A=<0 0 0>
     B=<0 0 -45>
     Alpha=&FN_FrontEndFloatModifier(Gameplay=0, Front End=$System.Timeline.NormalizedTime)
    )
   )
   &SubUVAnimation(
    "Sprite Renderer "=DataInterface() {propertyValues {SpriteRenderer=&self:Snowflakes_2.NiagaraSpriteRendererProperties_0}}
    SubUV Animation Mode:ENiagara_SubUVLookupModeV2=::NewEnumerator2
    UseEndFrame=false
    UseStartFrame=false
   )
  ]
  particleUpdateScript [
   &ParticleState()
   &Drag(Use Linear Drag=true, Use Rotational Drag=false)
   /Niagara/Modules/Update/Forces/V2/CurlNoiseForce(Noise Strength=250, Pan Noise=true, Pan Noise Field=<1 1 0>)
   &SolveForcesAndVelocity(Clamp Velocity=false, Limit Acceleration=false)
   &SpriteRotationRate(Rotation Rate=300)
   &ScaleColor(
    Scale Alpha=&FloatFromCurve(FloatCurve=DataInterface() {propertyValues {Curve {keys [{interpMode=RCIM_Cubic, tangentMode=RCTM_User, time=0.5, value=1}, {interpMode=RCIM_Cubic, tangentMode=RCTM_User, time=1, value=0}]}}})
    Scale RGB=&Multiply_VectorByFloat(Vector=<1 1 1>, Float=&FNTimeOfDayScalar(Night Scale=0.2, Morning Scale=1, Evening Scale=1))
    ScaleA=true
    ScaleRGB=true
   )
  ]
  rendererClasses [&NiagaraSpriteRendererProperties]
  renderers [{rendererClass=&NiagaraSpriteRendererProperties}]
 }
 TrailLong {
  emitterUpdateScript [
   &EmitterState(
    Inactive Response=::NewEnumerator1
    Life Cycle Mode=::NewEnumerator1
    Loop Behavior=::NewEnumerator1
    Loop Duration Mode=::NewEnumerator1
    UseLoopDelay=false
   )
   &SpawnBurst_Instantaneous(Spawn Count=2, Use Spawn Probability=false, UseLoopCountLimit=false)
   /SetVariables(Emitter.Texture_Speed_Master:NiagaraFloat=&MakeCustomFloatFromBool(
     Bool=&SetBoolByFloatComparison(A=1, B=$User.ParaGlideLeanAlpha, Comparison Type=::NewEnumerator4)
     True Float=0.7
     False Float=&RandomRangeFloat(Minimum=0.4, Maximum=0.5)
    )) {isSetParameters=true}
   /SetVariables(Emitter.Texture_Speed_Master_Detail:NiagaraFloat=&MakeCustomFloatFromBool(
     Bool=&SetBoolByFloatComparison(A=1, B=$User.ParaGlideLeanAlpha, Comparison Type=::NewEnumerator4)
     True Float=0.6
     False Float=0.45
    )) {isSetParameters=true}
   /SetVariables(Emitter.SphereMaskU:NiagaraFloat=&Lerp_Float(A=1, B=0.95, Alpha=$User.ParaGlideLeanAlpha)) {isSetParameters=true}
  ]
  particleSpawnScript [
   &PartitionParticles()
   &InitializeParticle(
    Color=<rgba 0 0.3129 0.6166 2.5>
    Mesh Scale=&SelectVectorFromArray(
     Direct Array Index=$Particles.PartitionParticles.Partition
     Vector Selection Array=DataInterface() {propertyValues {FloatData [<0.5 0.5 0.5>, <0.5 0.5 0.5>]}}
    )
    Mesh Scale Mode=::NewEnumerator4
    UsePositionOffset=true
    Write Material Random=true
   )
   &InitialMeshOrientation(Mesh Orientation Mode=::NewEnumerator2, Rotation=<0.25 0 0>, Use Rotation Vector=true)
   &InitialMeshOrientation(
    Mesh Orientation Mode=::NewEnumerator2
    Rotation=&SelectVectorFromArray(
     Direct Array Index=$Particles.PartitionParticles.Partition
     Vector Selection Array=DataInterface() {propertyValues {FloatData [<0 0 0>, <0 0 0.1>]}}
    )
    Use Rotation Vector=true
   ) {module=InitialMeshOrientation001}
  ]
  particleUpdateScript [
   &ParticleState(Kill Particles When Lifetime Has Elapsed=false, Let Infinitely Lived Particles Die When Emitter Deactivates=true)
   &ScaleMeshSize(Scale Factor=&VectorFromCurve(
     VectorCurve=DataInterface() {propertyValues {
       XCurve {defaultValue=0, keys [{tangentMode=RCTM_User, time=0, value=0}, {tangentMode=RCTM_User, time=1, value=1}]}
       YCurve {defaultValue=0, keys [{tangentMode=RCTM_User, time=0, value=0}, {tangentMode=RCTM_User, time=1, value=1}]}
       ZCurve {defaultValue=0, keys [{tangentMode=RCTM_User, time=0, value=0}, {tangentMode=RCTM_User, time=1, value=1}]}
      }}
     Scale Curve=&MakeCustomVectorFromBool(
      Bool=$User.bIsFrontEndPreview
      True Vector=&Lerp_Vector(A=<1 0.6 1>, B=<1 1.3 1>, Alpha=$User.ParaGlideLeanAlpha)
      False Vector=&Lerp_Vector(A=<1 0.9 1>, B=<1 1.5 1>, Alpha=$User.ParaGlideLeanAlpha)
     )
    ))
   &SkeletalMeshLocation(
    Bone / Socket Index=&SelectIntFromArray(Direct Array Index=$Particles.PartitionParticles.Partition, Int Selection Array=DataInterface() {propertyValues {IntData [0,1]}})
    Bone Sampling Mode=::NewEnumerator4
    MaskSpawnGroup=false
    Orientation Sampling=::NewEnumerator1
    Sampled Position Offset Coordinate Space:ENiagaraSkelMeshTransforms=::NewEnumerator2
    Skeletal Mesh=$System.Skeletal Mesh
   )
   &DynamicMaterialParameters(
    Index 0 Param 1=&RandomRangeFloat(Minimum=0.2, Maximum=1)
    Index 0 Param 2=&RandomRangeFloat(Minimum=0.2, Maximum=1)
    Index 0 Param 3=&FN_FrontEndFloatModifier(Gameplay=0.5, Front End=0)
    Index 0 Param 4=20
    Index 3 Param 1=150
    Index 3 Param 2=0.3
    Index 3 Param 4=&FloatFromCurve(FloatCurve=DataInterface() {propertyValues {Curve {keys [<at 0 0>,<at 1 1>]}}})
    Module.Index0Param0Write=true
    Module.Index0Param1Write=true
    Module.Index0Param2Write=true
    Module.Index0Param3Write=true
    Module.Index2Param0Write=true
    Module.Index2Param1Write=true
    Module.Index2Param2Write=true
    Module.Index2Param3Write=true
    Module.Index3Param0Write=true
    Module.Index3Param1Write=true
    Module.Index3Param2Write=true
    Module.Index3Param3Write=true
    Write Parameter Index 2=true
    Write Parameter Index 3=true
   )
   &InterpolateOverTime(
    Rate Of Change:NiagaraFloat=&FN_FrontEndFloatModifier(Gameplay=10, Front End=10)
    Target Vector Value:Vector3f=&AddVectorToPosition(
     Position=$Particles.Position
     Vector=&TransformVector(
      Vector=&FN_FrontEndVector3Modifier(Gameplay=$System.IGCenterposition, Front End=$System.FECenterposition)
      Destination Space=::Simulation
     )
    )
    Variable Type=::NewEnumerator2
   )
   /SetVariables(Particles.Interpolate Dynamic=&MakeFloat4FromVectorAndFloat(
     "Vector (RGB)"=$StackContext.InterpolateOverTime.InterpolateOverTime_V3_.Moving Average
     "Float (Alpha)"=&FN_FrontEndFloatModifier(Gameplay=1, Front End=0)
    )) {isSetParameters=true}
   /SetVariables(Particles.Color002=&Multiply_LinearColorByFloat(
     Linear Color=&MultiplyLinearColor(
      Linear Color=&MakeCustomLinearColorFromBool(Bool=$User.bIsFrontEndPreview, True Color=<rgba 0.0219 0.0346 0.0729 1>, False Color=<rgba 0.1 0.1 0.1 1>)
      Scale Factor=&MakeVector4_FromLinearColor(LinearColor=&FNTimeOfDayColor(Night Color=<rgba 0.1927 0.3945 1 1>, Morning Color=<rgba 0.6 0.6 0.6 1>, Day Color=<rgba 0.8 0.8 0.8 1>, Evening Color=<rgba 0.6 0.6 0.6 1>))
     )
     Float=&Multiply_Float(
      A=&FN_FrontEndFloatModifier(Gameplay=0.6, Front End=0.6)
      B=&FN_PickFloatDependOnPlatform(TargetPlatformActive=1, TargetPlatformNoActive=0.7)
     )
    )) {isSetParameters=true}
   /SetVariables(Particles.Size_Detail:Vector3f=&VectorFromCurve(VectorCurve=DataInterface() {propertyValues {
       XCurve {defaultValue=1, keys [{tangentMode=RCTM_User, time=0, value=0}, {tangentMode=RCTM_User, time=1, value=1}]}
       YCurve {defaultValue=1, keys [{tangentMode=RCTM_User, time=0, value=0}, {tangentMode=RCTM_User, time=1, value=1}]}
       ZCurve {defaultValue=1, keys [{tangentMode=RCTM_User, time=0, value=0}, {tangentMode=RCTM_User, time=1, value=1}]}
      }})) {isSetParameters=true}
   &ScaleColor(
    Scale Alpha=&FloatFromCurve(
     FloatCurve=DataInterface() {propertyValues {Curve {
        defaultValue=0
        keys [
         {arriveTangent=4.1785, interpMode=RCIM_Cubic, leaveTangent=4.1785, tangentMode=RCTM_User, time=0, value=0}
         {arriveTangent=1.1179, interpMode=RCIM_Cubic, leaveTangent=1.1179, tangentMode=RCTM_User, time=0.2042, value=0.6228}
         {arriveTangent=0.0224, interpMode=RCIM_Cubic, leaveTangent=0.0224, tangentMode=RCTM_User, time=1, value=1}
        ]
       }}}
     CurveIndex=$Emitter.Age
    )
    Scale RGB=&Multiply_Vector(
     A=&MakeVectorFromLinearColorRGB(Color=&FNTimeOfDayColor(Night Color=<rgba 0.5 0.625 1 1>, Morning Color=<rgba 1 1 1 1>, Day Color=<rgba 1 1 1 1>, Evening Color=<rgba 1 1 1 1>))
     B=&VectorFromFloat(Value=&FN_FrontEndFloatModifier(Gameplay=1, Front End=0.3))
    )
    ScaleA=true
    ScaleRGB=true
   )
   &UpdateMeshOrientation(
    Facing Interpolation=false
    Orientation Method=::NewEnumerator1
    Override Quat=false
    TransformFacing=false
    TransformUp=false
    Up Direction:Vector3f=&Lerp_Vector(
     A=<0 0 1>
     B=<1 -0.5 1>
     Alpha=&FN_FrontEndFloatModifier(Gameplay=0, Front End=$System.Timeline.NormalizedTime)
    )
   ) {module=UpdateMeshOrientation001}
  ]
  rendererClasses [&NiagaraMeshRendererProperties]
  renderers [
   {rendererClass=&NiagaraMeshRendererProperties, rendererIndex=0}
   {rendererClass=&NiagaraMeshRendererProperties, rendererIndex=1}
  ]
 }
 Trailspiral {
  emitterUpdateScript [
   &EmitterState(
    Inactive Response=::NewEnumerator1
    Life Cycle Mode=::NewEnumerator1
    Loop Behavior=::NewEnumerator1
    Loop Duration Mode=::NewEnumerator1
    UseLoopDelay=false
   )
   &SpawnRate(
    SpawnRate=&MakeCustomFloatFromBool(
     Bool=&SetBoolByFloatComparison(A=1, B=$User.ParaGlideLeanAlpha, Comparison Type=::NewEnumerator4)
     True Float=45
     False Float=30
    )
    Use Spawn Probability=false
   )
  ]
  particleSpawnScript [
   &PartitionParticles()
   &InitializeParticle(
    Color Maximum=<rgba 0.0254 0.0705 0.15 2>
    Color Minimum=<rgba 0.0254 0.0705 0.15 2>
    Color Mode=::NewEnumerator2
    Lifetime=&MakeCustomFloatFromBool(
     Bool=&SetBoolByFloatComparison(A=1, B=$User.ParaGlideLeanAlpha, Comparison Type=::NewEnumerator4)
     True Float=0.7
     False Float=1.2
    )
    Mesh Scale Max=<0.7 2 1.5>
    Mesh Scale Min=<0.4 1.5 1>
    Mesh Scale Mode=::NewEnumerator2
    UsePositionOffset=false
    Write Material Random=true
   )
   &InitialMeshOrientation(Mesh Orientation Mode=::NewEnumerator2, Rotation=<0 0.25 0>, Use Rotation Vector=true)
   &InitialMeshOrientation(
    Mesh Orientation Mode=::NewEnumerator2
    Rotation=&RandomRangeVector(Minimum=<0 0 -5>, Maximum=<0 0 5>)
    Use Rotation Vector=true
   ) {module=InitialMeshOrientation001}
   &ShapeLocation(
    Cylinder Height=&MakeCustomFloatFromBool(Bool=$User.bIsFrontEnd, True Float=10, False Float=10)
    Cylinder Radius=5
    MaskSpawnGroup=false
    Offset Mode=::NewEnumerator0
    Override Seed=false
    Rotation Coordinate Space=::Simulation
    Shape Primitive=::NewEnumerator1
    Surface Only=false
    UseOwnerScale=true
   )
   &AddVelocity(Velocity=&MakeCustomVectorFromBool(
     Bool=&SetBoolByFloatComparison(A=1, B=$User.ParaGlideLeanAlpha, Comparison Type=::NewEnumerator4)
     True Vector=<0 0 80>
     False Vector=<1 0 30>
    ))
  ]
  particleUpdateScript [
   &ParticleState()
   &SkeletalMeshLocation(
    Bone / Socket Index=&SelectIntFromArray(Direct Array Index=$Particles.PartitionParticles.Partition, Int Selection Array=DataInterface() {propertyValues {IntData [0,1]}})
    Bone Sampling Mode=::NewEnumerator4
    MaskSpawnGroup=false
    Orientation Sampling=::NewEnumerator1
    Sampled Position Offset=&FN_FrontEndVector3Modifier(Gameplay=<0 0 10>, Front End=<0 0 0>)
    Sampled Position Offset Coordinate Space:ENiagaraSkelMeshTransforms=::NewEnumerator2
    Skeletal Mesh=$System.Skeletal Mesh
   )
   &UpdateMeshOrientation(
    Facing Interpolation=false
    Orientation Method=::NewEnumerator1
    Override Quat=false
    TransformFacing=false
    TransformUp=false
    Up Direction:Vector3f=&Lerp_Vector(
     A=<0 0 1>
     B=<1 -0.5 1>
     Alpha=&FN_FrontEndFloatModifier(Gameplay=0, Front End=$System.Timeline.NormalizedTime)
    )
   )
   &ScaleMeshSize(Scale Factor=&MakeVector(
     X=1
     Y=&FloatFromCurve(FloatCurve=DataInterface() {propertyValues {Curve {keys [<at 0 0.5>,<at 0.3 0.85>,<at 0.8 1>,<at 1 1>]}}})
     Z=&FloatFromCurve(FloatCurve=DataInterface() {propertyValues {Curve {keys [<at 0 0.5>,<at 0.3 0.85>,<at 0.8 1>,<at 1 1>]}}})
    )) {module=ScaleMeshSize001}
   &ScaleMeshSize(Scale Factor=&Lerp_Vector(A=<1 1 1>, B=<1.3 1 1>, Alpha=$User.ParaGlideLeanAlpha))
   &DynamicMaterialParameters(
    Index 0 Param 2=-0.1
    Index 0 Param 3=1
    Index 0 Param 4=&FloatFromCurve(FloatCurve=DataInterface() {propertyValues {Curve {keys [<at 0 0>,<at 1 1>]}}})
    Index 3 Param 1=500
    Module.Index0Param0Write=true
    Module.Index0Param1Write=true
    Module.Index0Param2Write=true
    Module.Index0Param3Write=true
    Module.Index3Param0Write=true
    Module.Index3Param1Write=true
    Module.Index3Param2Write=true
    Module.Index3Param3Write=true
    Write Parameter Index 3=true
   )
   &ScaleColor(
    Scale Alpha=&FloatFromCurve(
     FloatCurve=DataInterface() {propertyValues {Curve {keys [
         {interpMode=RCIM_Cubic, tangentMode=RCTM_User, time=0.1, value=0}
         {arriveTangent=0.0424, interpMode=RCIM_Cubic, leaveTangent=0.0424, tangentMode=RCTM_User, time=0.3, value=0.8}
         {interpMode=RCIM_Cubic, tangentMode=RCTM_User, time=0.4, value=0.8}
         {tangentMode=RCTM_User, time=1, value=0}
        ]}}}
     Scale Curve=&Multiply_Float(
      A=0.5
      B=&FN_FrontEndFloatModifier(Gameplay=&RandomRangeFloat(Minimum=0.5, Maximum=1), Front End=0.8)
     )
    )
    ScaleA=true
    ScaleRGB=true
   )
   /SetVariables(Particles.Color002=&ColorFromCurve(ColorCurve=DataInterface() {propertyValues {
       AlphaCurve {keys [<at 0.006 0>,<at 0.4365 1>,<at 0.9896 0>]}
       BlueCurve {keys [<at 0.4142 0.005 RCIM_Cubic>,<at 1 1 RCIM_Cubic>]}
       GreenCurve {keys [<at 0.4142 0.0036 RCIM_Cubic>,<at 1 0.6398 RCIM_Cubic>]}
       RedCurve {keys [<at 0.4142 0.0024 RCIM_Cubic>,<at 1 0 RCIM_Cubic>]}
      }})) {isSetParameters=true}
   &SolveForcesAndVelocity(Clamp Velocity=false, Limit Acceleration=false)
  ]
  rendererClasses [&NiagaraMeshRendererProperties]
  renderers [
   {rendererClass=&NiagaraMeshRendererProperties, rendererIndex=0}
   {rendererClass=&NiagaraMeshRendererProperties, rendererIndex=1}
  ]
 }
}
summary {
 systemName=NS_Contrail_HollowStun
 emitters [
  {emitterName=FireTrail_And_Dots, rendererClasses [&NiagaraMeshRendererProperties]}
  {emitterName=Trailspiral, rendererClasses [&NiagaraMeshRendererProperties]}
  {emitterName=TrailLong, rendererClasses [&NiagaraMeshRendererProperties]}
  {emitterName=Ice_Frags, rendererClasses [&NiagaraMeshRendererProperties]}
  {emitterName=Snowflakes, rendererClasses [&NiagaraSpriteRendererProperties]}
 ]
 userVariables [
  User.bIsFrontEnd {type=NiagaraBool}
  User.bIsFrontEndPreview {type=NiagaraBool}
  User.bisInVehicle {type=NiagaraBool}
  User.IsSkydiving {type=NiagaraBool}
  User.ParaGlideLeanAlpha {type=NiagaraFloat, defaultValue=0, description=""}
 ]
}
